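# Niagara system for the ChillCat backpack cosmetic: greenish stink ribbon trails that follow the skeletal mesh and drift upward.
# Two active ribbon emitters crossfade by camera alignment with the owner's Z axis; an older third is disabled.
# A system moving value drives opacity tiling, pan speed and ribbon width.
# Anim-trail modules place particles, age-based curves fade and widen them, and inherited source movement tapers early in life.
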NiagaraSystem /BRCosmetics/Accessories/FORT_Backpacks/Backpack_F_MED_ChillCat/FX/NS_ChillCat_Backpack
dependencies {
 usedDataInterfaces [&NiagaraDataInterfaceCamera, &NiagaraDataInterfaceCurve]
 usedDynamicInputs [
  &MakeIntFromFloat
  &MakeCustomFloatFromBool
  &Boolean_OrOperation
  &SetBoolByFloatComparison
  &MakeFloatFromVector
 ]
 usedModules [
  &self:SystemSpawnScript.SystemScriptSource.SystemScriptGraph.NiagaraNodeAssignment_0.SetVariables_#1
  &SystemState
  &self:SystemSpawnScript.SystemScriptSource.SystemScriptGraph.NiagaraNodeAssignment_1.SetVariables_#2
 ]
}
emitters {
 Stink_Ribbon {
  bEnabled=false
  emitterSpawnScript [&FN_RibbonIDOffsetUtility()]
  emitterUpdateScript [
   &EmitterState()
   &FN_SpawnAnimTrail(Spawn Interval=0)
  ]
  particleSpawnScript [
   &FN_InitializeAnimTrailBoneIndecies()
   &FN_InitializeAnimTrail(Lifetime=0.85)
   &AddVelocity(Additional Quaternion Rotation=false, Rotation Mode=::NewEnumerator3, Rotation Quaternion:Quat4f=$Engine.Owner.Rotation, Velocity=<0 -35 70>)
  ]
  particleUpdateScript [
   &ParticleState()
   /SetVariables(Particles.OpacityScale:NiagaraFloat=&Lerp_Float(A=3, B=1.25, Alpha=$User.IsMoving)) {isSetParameters=true, module=SetVariables_#3}
   /SetVariables(Particles.OpacityPan:NiagaraFloat=&Lerp_Float(A=0.65, B=2.2, Alpha=$User.IsMoving)) {isSetParameters=true, module=SetVariables_#4}
   &InterpolateOverTime(Rate Of Change=50, Target Float Value:NiagaraFloat=$Particles.OpacityScale)
   &InterpolateOverTime(Rate Of Change=50, Target Float Value:NiagaraFloat=$Particles.OpacityPan) {module=InterpolateOverTime001}
   /SetVariables(Particles.CameraFloat:NiagaraFloat=&Abs_Float(Float:NiagaraFloat=&DotProduct(
      Vector A=&NormalizeVector(Vector To Normalize=$Engine.Owner.SystemXAxis, Safe Normalize=true, Safe Fallback Threshold=0, Safe Fallback Vector=<0 0 0.001>)
      Vector B=&NormalizeVector(
       Vector To Normalize=&ReturnCameraVector(Camera Vectors=::NewEnumerator1, Coordinate Space=::Simulation)
       Safe Normalize=true
       Safe Fallback Threshold=0
       Safe Fallback Vector=<0 0 0.001>
      )
     ))) {isSetParameters=true, module=SetVariables_#5}
   &InheritSourceMovement(
    Applied Position Scale=&VectorFromFloat(Value=&FloatFromCurve(FloatCurve=DataInterface() {propertyValues {Curve {keys [
          {arriveTangent=-0.3198, arriveTangentWeight=0.2288, interpMode=RCIM_Constant, leaveTangentWeight=0.2288, time=0, value=1}
          {arriveTangent=-1.3333, interpMode=RCIM_Cubic, leaveTangent=-1.3333, time=0.1, value=1}
          {interpMode=RCIM_Cubic, leaveTangent=-1.1111, time=0.3, value=0.6}
         ]}}}))
    ApplyForce=false
    ApplyPosition=true
    ApplyPositionFalloff=false
   )
   &Drag(Drag=0.5, Rotational Drag=50, Use Linear Drag=true, Use Rotational Drag=true)
   &GravityForce(Gravity=<0 0 120>)
   &SolveForcesAndVelocity(Clamp Velocity=false, Limit Acceleration=false)
   &FN_UpdateAnimTrail(SkeletalMesh=$User.SkeletalMesh) {module=FN_UpdateAnimTrail001}
   &FN_RibbonMaterial(
    Color=<rgba 0.1112 0.125 0.0456 1>
    Macro Distortion Amount=0.085
    Macro Pan Speed=0.05
    Macro Scale U=0.1
    Opacity Pan U Offset:NiagaraFloat=/Game/Effects/Niagara/DynamicInputs/Random/FN_SeededRandomRangeFloat(Minimum:NiagaraFloat=0, Maximum:NiagaraFloat=10, Random Seed:NiagaraInt32=$Engine.System.RandomSeed)
    Opacity Pan U Speed:NiagaraFloat=$Particles.InterpolateOverTime001.InterpolateOverTime_Float.Moving Average
    Opacity Power=3
    Opacity Scale U:NiagaraFloat=$Particles.InterpolateOverTime.InterpolateOverTime_Float.Moving Average
    Scale Alpha:NiagaraFloat=&FloatFromCurve(FloatCurve=DataInterface() {propertyValues {Curve {keys [
         {arriveTangent=-1.4491, interpMode=RCIM_Cubic, time=0.25, value=1}
         {arriveTangent=-3.136, interpMode=RCIM_Cubic, leaveTangent=-3.136, tangentMode=RCTM_User, time=0.95, value=0}
        ]}}})
    Scale Color:Vector3f=<5 5 5>
   )
   &ScaleRibbonWidth(Ribbon Width Scale=&Lerp_Float(
     A=&Lerp_Float(A=0.7, B=0, Alpha=$Particles.CameraFloat)
     B=4
     Alpha=&FloatFromCurve(
      FloatCurve=DataInterface() {propertyValues {Curve {keys [<at 0 0 RCIM_Cubic>, {interpMode=RCIM_Cubic, leaveTangent=0.9894, time=0.9, value=1}]}}}
      CurveIndex=$Particles.RibbonU0Override
     )
    ))
   /SetVariables(Particles.RibbonLinkOrder:NiagaraFloat=$Particles.NormalizedAge) {isSetParameters=true, module=SetVariables_#6}
   /SetVariables(Particles.CustomFacing:Vector3f=&Lerp_Vector(
     A=$Engine.Owner.SystemZAxis
     B=&ReturnCameraVector(Camera Vectors=::NewEnumerator1, Coordinate Space=::Simulation)
     Alpha=$Particles.CameraFloat
    )) {enabled=false, isSetParameters=true, module=SetVariables_#7}
  ]
  rendererClasses [&NiagaraRibbonRendererProperties]
  renderers [{rendererClass=&NiagaraRibbonRendererProperties}]
 }
 Stink_Ribbon001 {
  emitterSpawnScript [&FN_RibbonIDOffsetUtility()]
  emitterUpdateScript [
   &EmitterState(Life Cycle Mode=::NewEnumerator1, Loop Behavior=::NewEnumerator1, Loop Duration Mode=::NewEnumerator1, UseLoopDelay=false)
   &FN_SpawnAnimTrail(Spawn Interval=0)
  ]
  particleSpawnScript [
   &FN_InitializeAnimTrailBoneIndecies()
   &FN_InitializeAnimTrail(Lifetime=0.75, Position Offset=<0 0 -0.5>)
   &AddVelocity(Additional Quaternion Rotation=false, Rotation Mode=::NewEnumerator3, Rotation Quaternion:Quat4f=$Engine.Owner.Rotation, Velocity=<0 -35 60>)
  ]
  particleUpdateScript [
   &ParticleState()
   /SetVariables(Particles.OpacityScale:NiagaraFloat=&Lerp_Float(A=3, B=1, Alpha=$System.IsMoving)) {isSetParameters=true, module=SetVariables_#3}
   /SetVariables(Particles.OpacityPan:NiagaraFloat=&Lerp_Float(A=0.25, B=2, Alpha=$System.IsMoving)) {isSetParameters=true, module=SetVariables_#4}
   &InterpolateOverTime(Rate Of Change=50, Target Float Value:NiagaraFloat=$Particles.OpacityScale)
   &InterpolateOverTime(Rate Of Change=50, Target Float Value:NiagaraFloat=$Particles.OpacityPan) {module=InterpolateOverTime001}
   /SetVariables(
    Particles.CameraFloat:NiagaraFloat=&Abs_Float(Float:NiagaraFloat=&DotProduct(
      Vector A=&NormalizeVector(Vector To Normalize=$Engine.Owner.SystemZAxis, Safe Normalize=true, Safe Fallback Threshold=0, Safe Fallback Vector=<0 0 0.001>)
      Vector B=&NormalizeVector(
       Vector To Normalize=&ReturnCameraVector(Camera Vectors=::NewEnumerator1, Coordinate Space=::Simulation)
       Safe Normalize=true
       Safe Fallback Threshold=0
       Safe Fallback Vector=<0 0 0.001>
      )
     ))
    Particles.CameraShrinkFloat:NiagaraFloat=&Abs_Float(Float:NiagaraFloat=&DotProduct(
      Vector A=&NormalizeVector(Vector To Normalize=$Engine.Owner.SystemXAxis, Safe Normalize=true, Safe Fallback Threshold=0, Safe Fallback Vector=<0 0 0.001>)
      Vector B=&NormalizeVector(
       Vector To Normalize=&ReturnCameraVector(Camera Vectors=::NewEnumerator1, Coordinate Space=::Simulation)
       Safe Normalize=true
       Safe Fallback Threshold=0
       Safe Fallback Vector=<0 0 0.001>
      )
     ))
   ) {isSetParameters=true, module=SetVariables_#5}
   &InheritSourceMovement(
    Applied Position Scale=&VectorFromFloat(Value=&FloatFromCurve(FloatCurve=DataInterface() {propertyValues {Curve {keys [
          {
           arriveTangent=-3.5444
           arriveTangentWeight=0.2288
           interpMode=RCIM_Cubic
           leaveTangent=-3.5444
           leaveTangentWeight=0.2288
           tangentMode=RCTM_User
           time=0
           value=1
          }
          {interpMode=RCIM_Cubic, leaveTangent=-4.5, time=0.3, value=0.2}
         ]}}}))
    ApplyForce=false
    ApplyPosition=true
    ApplyPositionFalloff=false
   )
   &Drag(Use Linear Drag=true, Use Rotational Drag=true)
   &GravityForce(Gravity=<0 0 120>)
   &SolveForcesAndVelocity(Clamp Velocity=false, Limit Acceleration=false)
   &FN_UpdateAnimTrail(SkeletalMesh=$User.SkeletalMesh) {module=FN_UpdateAnimTrail001}
   &FN_RibbonMaterial(
    Color=<rgba 0.1112 0.125 0.0456 1>
    Macro Distortion Amount=0.05
    Macro Pan Speed=0.05
    Macro Scale U=0.1
    Opacity Pan U Offset:NiagaraFloat=&MakeFloatFromInt(Int=$System.RandomSeed)
    Opacity Pan U Speed:NiagaraFloat=$Particles.InterpolateOverTime001.InterpolateOverTime_Float.Moving Average
    Opacity Power=3
    Opacity Scale U:NiagaraFloat=$Particles.InterpolateOverTime.InterpolateOverTime_Float.Moving Average
    Scale Alpha:NiagaraFloat=&FloatFromCurve(
     FloatCurve=DataInterface() {propertyValues {Curve {keys [
         {arriveTangent=-1.4491, interpMode=RCIM_Cubic, time=0.25, value=1}
         {arriveTangent=-3.136, interpMode=RCIM_Cubic, leaveTangent=-3.136, tangentMode=RCTM_User, time=0.95, value=0}
        ]}}}
     Scale Curve=&Lerp_Float(
      A=1
      B=0
      Alpha=&FloatFromCurve(
       FloatCurve=DataInterface() {propertyValues {Curve {keys [
           {arriveTangent=12.9678, interpMode=RCIM_Cubic, time=0.5, value=0}
           {arriveTangent=3.4241, interpMode=RCIM_Cubic, leaveTangent=3.4241, tangentMode=RCTM_User, time=0.9, value=1}
          ]}}}
       CurveIndex=$Particles.CameraFloat
      )
     )
    )
    Scale Color:Vector3f=<5 5 5>
   )
   &ScaleRibbonWidth(Ribbon Width Scale=&Lerp_Float(
     A=&Lerp_Float(A=0.7, B=0, Alpha=$Particles.CameraShrinkFloat)
     B=&Lerp_Float(A=4, B=6, Alpha=$System.IsMoving)
     Alpha=&FloatFromCurve(FloatCurve=DataInterface() {propertyValues {Curve {keys [<at 0 0 RCIM_Cubic>, {interpMode=RCIM_Cubic, leaveTangent=0.9894, time=0.9, value=1}]}}})
    ))
   /SetVariables(Particles.RibbonLinkOrder:NiagaraFloat=$Particles.NormalizedAge) {isSetParameters=true, module=SetVariables_#6}
   /SetVariables(Particles.CustomFacing:Vector3f=&Lerp_Vector(
     A=&ReturnCameraVector(Camera Vectors=::NewEnumerator1, Coordinate Space=::Simulation)
     B=&Lerp_Vector(
      A=$Engine.Owner.SystemZAxis
      B=&ReturnCameraVector(Camera Vectors=::NewEnumerator1, Coordinate Space=::Simulation)
      Alpha=&FloatFromCurve(FloatCurve=DataInterface() {propertyValues {Curve {keys [
           {arriveTangent=3.3777, interpMode=RCIM_Cubic, time=0, value=0}
           {arriveTangent=12.6465, interpMode=RCIM_Cubic, leaveTangent=12.6465, tangentMode=RCTM_User, time=0.2, value=1}
          ]}}})
     )
     Alpha=$User.IsMoving
    )) {enabled=false, isSetParameters=true, module=SetVariables_#7}
  ]
  rendererClasses [&NiagaraRibbonRendererProperties]
  renderers [{rendererClass=&NiagaraRibbonRendererProperties}]
 }
 Stink_Ribbon002 {
  emitterSpawnScript [&FN_RibbonIDOffsetUtility()]
  emitterUpdateScript [
   &EmitterState(Life Cycle Mode=::NewEnumerator1, Loop Behavior=::NewEnumerator1, Loop Duration Mode=::NewEnumerator1, UseLoopDelay=false)
   &FN_SpawnAnimTrail()
  ]
  particleSpawnScript [
   &FN_InitializeAnimTrailBoneIndecies()
   &FN_InitializeAnimTrail(Lifetime=0.75, Position Offset=<0 0 -1>)
   &AddVelocity(Additional Quaternion Rotation=false, Rotation Mode=::NewEnumerator3, Rotation Quaternion:Quat4f=$Engine.Owner.Rotation, Velocity=<0 -35 60>)
  ]
  particleUpdateScript [
   &ParticleState()
   /SetVariables(Particles.OpacityScale:NiagaraFloat=&Lerp_Float(A=3, B=1, Alpha=$System.IsMoving)) {isSetParameters=true, module=SetVariables_#3}
   /SetVariables(Particles.OpacityPan:NiagaraFloat=&Lerp_Float(A=0.25, B=2, Alpha=$System.IsMoving)) {isSetParameters=true, module=SetVariables_#4}
   &InterpolateOverTime(Rate Of Change=50, Target Float Value:NiagaraFloat=$Particles.OpacityScale)
   &InterpolateOverTime(Rate Of Change=50, Target Float Value:NiagaraFloat=$Particles.OpacityPan) {module=InterpolateOverTime001}
   /SetVariables(Particles.CameraFloat:NiagaraFloat=&Abs_Float(Float:NiagaraFloat=&DotProduct(
      Vector A=&NormalizeVector(Vector To Normalize=$Engine.Owner.SystemZAxis, Safe Normalize=true, Safe Fallback Threshold=0, Safe Fallback Vector=<0 0 0.001>)
      Vector B=&NormalizeVector(
       Vector To Normalize=&ReturnCameraVector(Camera Vectors=::NewEnumerator1, Coordinate Space=::Simulation)
       Safe Normalize=true
       Safe Fallback Threshold=0
       Safe Fallback Vector=<0 0 0.001>
      )
     ))) {isSetParameters=true, module=SetVariables_#5}
   &InheritSourceMovement(
    Applied Position Scale=&VectorFromFloat(Value=&FloatFromCurve(FloatCurve=DataInterface() {propertyValues {Curve {keys [
          {arriveTangent=-10.25, arriveTangentWeight=0.2288, interpMode=RCIM_Cubic, leaveTangentWeight=0.2288, time=0, value=1}
          {arriveTangent=-30.8583, interpMode=RCIM_Cubic, leaveTangent=-4.1408, tangentMode=RCTM_Break, time=0.1, value=0.2}
          <at 0.2 0.1>
         ]}}}))
    ApplyForce=false
    ApplyPosition=true
    ApplyPositionFalloff=false
   )
   &Drag(Use Linear Drag=true, Use Rotational Drag=true)
   &GravityForce(Gravity=<0 0 120>)
   &SolveForcesAndVelocity(Clamp Velocity=false, Limit Acceleration=false)
   &FN_UpdateAnimTrail(SkeletalMesh=$User.SkeletalMesh) {module=FN_UpdateAnimTrail001}
   &FN_RibbonMaterial(
    Color=<rgba 0.1112 0.125 0.0456 1>
    Macro Distortion Amount=0.05
    Macro Pan Speed=0.05
    Macro Scale U=0.1
    Opacity Pan U Offset:NiagaraFloat=&MakeFloatFromInt(Int=$System.RandomSeed)
    Opacity Pan U Speed:NiagaraFloat=$Particles.InterpolateOverTime001.InterpolateOverTime_Float.Moving Average
    Opacity Power=3
    Opacity Scale U:NiagaraFloat=$Particles.InterpolateOverTime.InterpolateOverTime_Float.Moving Average
    Scale Alpha:NiagaraFloat=&FloatFromCurve(
     FloatCurve=DataInterface() {propertyValues {Curve {keys [
         {arriveTangent=-1.4491, interpMode=RCIM_Cubic, time=0.25, value=1}
         {arriveTangent=-3.136, interpMode=RCIM_Cubic, leaveTangent=-3.136, tangentMode=RCTM_User, time=0.95, value=0}
        ]}}}
     Scale Curve=&Lerp_Float(
      A=0
      B=1
      Alpha=&FloatFromCurve(
       FloatCurve=DataInterface() {propertyValues {Curve {keys [{arriveTangent=10.0347, interpMode=RCIM_Cubic, leaveTangent=10.0347, tangentMode=RCTM_User, time=0.6, value=0}, <at 0.9 1 RCIM_Cubic>]}}}
       CurveIndex=$Particles.CameraFloat
      )
     )
    )
    Scale Color:Vector3f=<5 5 5>
   )
   &ScaleRibbonWidth(Ribbon Width Scale=&Lerp_Float(
     A=0.7
     B=&Lerp_Float(A=4, B=6, Alpha=$System.IsMoving)
     Alpha=&FloatFromCurve(FloatCurve=DataInterface() {propertyValues {Curve {keys [<at 0 0 RCIM_Cubic>, {interpMode=RCIM_Cubic, leaveTangent=0.9894, time=0.9, value=1}]}}})
    ))
   /SetVariables(Particles.RibbonLinkOrder:NiagaraFloat=$Particles.NormalizedAge) {isSetParameters=true, module=SetVariables_#6}
  ]
  rendererClasses [&NiagaraRibbonRendererProperties]
  renderers [{rendererClass=&NiagaraRibbonRendererProperties}]
 }
}
summary {
 systemName=NS_ChillCat_Backpack
 emitters [
  {bEnabled=false, emitterName=Stink_Ribbon, rendererClasses [&NiagaraRibbonRendererProperties]}
  {emitterName=Stink_Ribbon001, rendererClasses [&NiagaraRibbonRendererProperties]}
  {emitterName=Stink_Ribbon002, rendererClasses [&NiagaraRibbonRendererProperties]}
 ]
 userVariables [
  User.IsMoving {type=NiagaraFloat, defaultValue=0, description=""}
  User.RibbonIDOffset {type=NiagaraInt32}
  User.Running {type=NiagaraBool}
  User.SkeletalMesh {
   type=NiagaraDataInterfaceSkeletalMesh
   defaultValue {
    struct=NiagaraExt_VariableValue_DataInterface
    value {dataInterface=&self:NiagaraDataInterfaceSkeletalMesh_0}
   }
  }
 ]
}
