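# Physics asset for the F_MED_ChillCat backpack cosmetic, covering the pack and its two dynamic grenade pin bones.
# One large capsule stands in for the pack, and each grenade pin bone gets a small capsule of its own.
# Each pin is constrained directly to the pack with narrow swing limits and an almost locked twist, restricting it to a slight wobble.
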PhysicsAsset /BRCosmetics/Accessories/FORT_Backpacks/Backpack_F_MED_ChillCat/Meshes/F_MED_ChillCat_Pack_Physics
bodies {
 grenade_pin_dyn01 [{
   shapeType=Capsule
   length=1.4229
   radius=1.124
   shapeName=grenade_pin_dyn01_capsule
   center=<1.4026 -0.0137 -0.0622>
   rotation=<rot 0 -90.5901 -87.4153>
  }]
 grenade_pin_dyn02 [{
   shapeType=Capsule
   length=1.4294
   radius=1.254
   shapeName=grenade_pin_dyn02_capsule
   center=<1.453 -0.0192 0.2022>
   rotation=<rot 0 89.9234 -82.7112>
  }]
 pack [{shapeType=Capsule, length=10, radius=10, rotation=<rot 0.0144 90 -75.1914>}]
}
constraints [
 {bone1Name=grenade_pin_dyn01, bone2Name=pack, swing1LimitDegrees=8, swing2LimitDegrees=8, twistLimitDegrees=1}
 {bone1Name=grenade_pin_dyn02, bone2Name=pack, swing1LimitDegrees=8, swing2LimitDegrees=8, twistLimitDegrees=1}
]
